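# Example Default Lit water material that layers a tiling animated wave texture with the fluid simulation through the WaterSimulationTexture function.
# A world XY-projected normal-with-height texture runs through Motion_4WayChaos, then is rebiased to signed range.
# Red and green become a derived, flattened normal, while blue height scales a vector for world position offset, both fed into the simulation function.
# Simple parameters define base color, roughness and specular.
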Material /FortWater/FluidSimulation/Materials/Examples/M_Simple_Water_With_Sim
graph {
 properties {pixelDepthOffsetMode=PDOM_Legacy, usedShadingModels=Default Lit}
 nodes [
  n1 = ScalarParameter() {defaultValue=1000, parameterName=Normal Size}
  n2 = TextureObject() {samplerType=SAMPLERTYPE_LinearColor, texture=/Game/Global/Textures/Default/Tileable_Patterns/T_Water_TilingNormal_With_Height_02_Softened}
  n3 = ConstantBiasScale(None=MaterialFunctionCall!(
    Coordinates=MaterialFunctionCall!(Scale=@n1) {materialFunction=/Engine/Functions/Engine_MaterialFunctions01/Coordinates/WorldPosition-XY}.Result
    Speed=ScalarParameter() {defaultValue=0.5, parameterName=Speed}
    Divisor
    Texture=@n2
   ) {materialFunction=/Engine/Functions/Engine_MaterialFunctions03/ComplexMotion/Motion_4WayChaos}.Result) {bias=-0.5, scale=2}
  n4 = ComponentMask(None=@n3) {g=true, r=true}
  MaterialFunctionCall!(
   Flowmap Velocity
   Previous Foam
   Normal=MaterialFunctionCall(Normal=DeriveNormalZ(InXY=@n4), Flatness=ScalarParameter() {defaultValue=0.95, parameterName=Flatten}) {materialFunction=&FlattenNormal}.Result
   WPO=Multiply(A=ComponentMask(None=@n3) {b=true}, B=VectorParameter() {parameterName=Texture WPO, defaultValue=<rgba 0 0 4 0>}.RGB)
  ) {materialFunction=/FortWater/Materials/Functions/WaterSimulationTexture}
  ScalarParameter() {defaultValue=0.05, parameterName=Roughness}
  ScalarParameter() {defaultValue=0.225, parameterName=Spec}
  VectorParameter() {parameterName=Base, defaultValue=<rgba 0 0.0282 0.06 0>}
 ]
}
